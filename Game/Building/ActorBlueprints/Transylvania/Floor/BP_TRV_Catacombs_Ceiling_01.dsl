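; Transylvania catacombs ceiling building piece that occasionally carries an ambient bug particle effect.
; On BeginPlay a random roll gives a 10% chance to spawn the bugs emitter at an actor-relative offset with a flipped, world-space rotation, store it in Bugs and limit its draw distance.
; Damage effects deactivate the stored emitter without checking that it exists, and the Destroyed event only calls DestroyActor on itself.
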(event EventOnDamagePlayEffects
  (params (Damage :type Float%20%28single-precision%29) (DamageTags :type Gameplay%20Tag%20Container%20Structure%20%28by%20ref%29) (Momentum :type Vector :default "0, 0, 0") (HitInfo :type Hit%20Result%20Structure%20%28by%20ref%29) (InstigatedBy :type Fort%20Pawn%20Object%20Reference) (DamageCauser :type Actor%20Object%20Reference) (EffectContext :type Gameplay%20Effect%20Context%20Handle%20Structure))
  (Components|Activation|Deactivate (Variables|Default|GetBugs)))

(event EventDestroyed
  (Actor|DestroyActor))

(event EventBeginPlay
  (if (< (Math|Random|RandomFloatinRange 0.0 100) 10)
    (bind _returnvalue (Effects|SpawnEmitteratLocation "/Game/Packages/Fortress_Transylvania/Transylvania_effects/Effects/P_Bugs_01.P_Bugs_01" (Math|Transform|TransformLocation (Transformation|GetActorTransform) (Math|Vector|MakeVector (.x "0, 0, 0") (+ (.y "0, 0, 0") 245) (- (.z "0, 0, 0") 3))) (Math|Rotator|MakeRotator (.roll "0, 0, 0") (+ (.pitch "0, 0, 0") 180) (.yaw "0, 0, 0")) "1.000000,1.000000,1.000000" false))
    (Variables|Default|SetBugs _returnvalue)
    (LOD|SetMaxDrawDistance (Variables|Default|GetBugs) 1000)))

(fn ConstructionScript)
