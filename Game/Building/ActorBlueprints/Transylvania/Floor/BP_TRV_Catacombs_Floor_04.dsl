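; Blueprint for a Transylvania catacombs floor building piece, variant 04.
; It overrides GetInvalidTrapTagQuery to return a fixed gameplay tag query matching anything tagged Trap, which by its name marks all traps as invalid for this floor.
; The construction script is empty and the file shows no other logic.
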(fn GetInvalidTrapTagQuery
  (return "(TokenStreamVersion=0,TagDictionary=((TagName=\"Trap\")),QueryTokenStream=(0,1,1,1,0),UserDescription=\"\",AutoDescription=\" ANY( Trap )\")"))

(fn ConstructionScript)
